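# Idle Niagara effect for the SprintLuckSlaw pickaxe, layering a cyan mesh glow with star flashes and drifting sparkle sprites.
# FanGlow's scale, glow and material parameters follow system Timelines, while StarFlash bursts a large flare plus curl-noise sparks at a socket.
# Stars spawn while User.PowersActive is true, track socket offsets in three spawn groups, and adjust size and brightness by quality level, with audio-reactive corona sizing.
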NiagaraSystem /BRCosmetics/Weapons/FORT_Melee/Pickaxe_SprintLuckSlaw/FX/NS_Pickaxe_SprintLuckSlaw_Idle
dependencies {
 usedDataInterfaces [&NiagaraDataInterfaceCurve, &NiagaraDataInterfaceArrayFloat, &NiagaraDataInterfaceSocketReader]
 usedDynamicInputs [
  &RandomRangeLinearColor
  &MakeVector2D
  &SetBoolByIntComparison
 ]
 usedModules [
  &self:SystemSpawnScript.SystemScriptSource.SystemScriptGraph.NiagaraNodeAssignment_0.SetVariables_#1
  &self:SystemSpawnScript.SystemScriptSource.SystemScriptGraph.NiagaraNodeAssignment_1.SetVariables_#2
  &SystemState
  &self:SystemSpawnScript.SystemScriptSource.SystemScriptGraph.NiagaraNodeAssignment_2.SetVariables_#3
  &Timeline
  &self:SystemSpawnScript.SystemScriptSource.SystemScriptGraph.NiagaraNodeAssignment_3.SetVariables_#4
 ]
}
emitters {
 FanGlow {
  emitterUpdateScript [
   &EmitterState(Life Cycle Mode=::NewEnumerator1, Loop Behavior=::NewEnumerator1, Loop Duration Mode=::NewEnumerator1, UseLoopDelay=false)
   &SpawnBurst_Instantaneous(Use Spawn Probability=false, UseLoopCountLimit=false)
   /Niagara/Modules/Emitter/SpawnByTrigger(Trigger:NiagaraBool=$User.PowersActive, TriggerIntervalSwitch=false) {enabled=false}
  ]
  particleSpawnScript [
   &InitializeParticle(Color=<rgba 0.3281 0.725 1 1>, Lifetime=5, Mesh Scale Mode=::NewEnumerator3, UsePositionOffset=false, Write Material Random=true)
   &CameraOffset(Camera Offset Amount=6)
  ]
  particleUpdateScript [
   &ParticleState(
    Kill Particles When Lifetime Has Elapsed=false
    Let Infinitely Lived Particles Die When Emitter Deactivates=true
    Loop Particles Lifetime=true
   )
   &ScaleMeshSize(Scale Factor=&VectorFromFloat(Value=&FloatFromCurve(
      FloatCurve=DataInterface() {propertyValues {Curve {keys [
          <at 0 0 RCIM_Cubic>
          {arriveTangent=-0.037, time=0.01, value=1}
          {arriveTangent=-0.0216, interpMode=RCIM_Cubic, leaveTangent=-0.0216, tangentMode=RCTM_User, time=0.99, value=1}
          <at 1 0>
         ]}}}
      CurveIndex=$System.Timeline.Time
     )))
   /SetVariables(Particles.GlowColor=&Multiply_LinearColorByFloat(
     Linear Color=<rgba 0.0009 1.7725 3 1>
     Float=&Multiply_Float(
      A=&FloatFromCurve(
       FloatCurve=DataInterface() {propertyValues {Curve {keys [
           <at 0.0476 0>
           {arriveTangent=0.0091, interpMode=RCIM_Cubic, leaveTangent=0.0091, tangentMode=RCTM_User, time=0.09, value=1}
           {arriveTangent=-0.0216, interpMode=RCIM_Cubic, leaveTangent=-0.0216, tangentMode=RCTM_User, time=0.8, value=1}
           <at 1 0>
          ]}}}
       CurveIndex=$System.Timeline.Time
       Scale Curve=0.3
      )
      B=1
     )
    )) {isSetParameters=true}
   &ScaleColor(
    Scale Alpha=&FloatFromCurve(
     FloatCurve=DataInterface() {propertyValues {Curve {keys [
         <at 0.0476 0>
         {arriveTangent=0.0091, interpMode=RCIM_Cubic, leaveTangent=0.0091, tangentMode=RCTM_User, time=0.09, value=1}
         {arriveTangent=-0.0216, interpMode=RCIM_Cubic, leaveTangent=-0.0216, tangentMode=RCTM_User, time=0.8, value=1}
         <at 1 0>
        ]}}}
     CurveIndex=$System.Timeline.Time
    )
    ScaleA=true
    ScaleRGB=true
   ) {module=ScaleColor001}
   &DynamicMaterialParameters(
    Index 0 Param 1=&Lerp_Float(
     A=&FloatFromCurve(
      FloatCurve=DataInterface() {propertyValues {Curve {keys [
          <at 0.0476 -1 RCIM_Cubic>
          {arriveTangent=0.0091, interpMode=RCIM_Cubic, leaveTangent=0.0091, tangentMode=RCTM_User, time=0.22, value=-0.1}
          {arriveTangent=-0.0216, interpMode=RCIM_Cubic, leaveTangent=-0.0216, tangentMode=RCTM_User, time=0.8034, value=-0.1}
          <at 1 -1>
         ]}}}
      CurveIndex=$System.Timeline.Time
     )
     B=-0.5
     Alpha=$System.Timeline001.Time
    )
    Module.Index0Float0Write=true
    Module.Index0Float1Write=true
    Module.Index0Float2Write=true
    Module.Index0Float3Write=true
   )
  ]
  rendererClasses [&NiagaraMeshRendererProperties]
  renderers [
   {rendererClass=&NiagaraMeshRendererProperties, rendererIndex=0}
   {rendererClass=&NiagaraMeshRendererProperties, rendererIndex=1}
  ]
 }
 StarFlash {
  emitterUpdateScript [
   &EmitterState(Life Cycle Mode=::NewEnumerator1, Loop Behavior=::NewEnumerator1, Loop Delay=0.15, Loop Duration=1, UseLoopDelay=false)
   &SpawnBurst_Instantaneous(Use Spawn Probability=false, UseLoopCountLimit=false) {module=SpawnBurst_Instantaneous001}
   &SpawnBurst_Instantaneous(
    Spawn Count=&RandomRangeInt(Minimum=10, Maximum=15)
    Spawn Group=1
    Use Spawn Probability=false
    UseLoopCountLimit=false
   )
  ]
  particleSpawnScript [
   /SetVariables(Particles.VisibilityTag=$Emitter.SpawnGroup) {isSetParameters=true}
   &InitializeParticle(
    Color=<rgba 0.3 1.29 2 1>
    Lifetime=&MergeHelper_FloatSwitch(
     0=0.2
     1=&RandomRangeFloat(Minimum=0.3, Maximum=0.8)
     2=?
     3=?
     Index=$Emitter.SpawnGroup
     Num=::NewEnumerator0
    )
    Sprite Size Mode=::NewEnumerator3
    Uniform Sprite Size=&MergeHelper_FloatSwitch(
     0=200
     1=&RandomRangeFloat(Minimum=10, Maximum=25)
     2=?
     3=?
     Index=$Emitter.SpawnGroup
     Num=::NewEnumerator0
    )
    UsePositionOffset=false
    Write Material Random=true
   )
   &DynamicMaterialParameters(
    Index 0 Param 1=&MergeHelper_FloatSwitch(0=1, 1=0.1, 2=?, 3=?, Index=$Emitter.SpawnGroup, Num=::NewEnumerator0)
    Index 0 Param 2=&MergeHelper_FloatSwitch(
     0=0
     1=&RandomRangeFloat(Minimum=0.25, Maximum=0.5)
     2=?
     3=?
     Index=$Emitter.SpawnGroup
     Num=::NewEnumerator0
    )
    Index 0 Param 3=&MergeHelper_FloatSwitch(
     0=0
     1=&RandomRangeFloat(Minimum=0.25, Maximum=0.5)
     2=?
     3=?
     Index=$Emitter.SpawnGroup
     Num=::NewEnumerator0
    )
    Index 0 Param 4=1
    Module.Index0Float0Write=true
    Module.Index0Float1Write=true
    Module.Index0Float2Write=true
    Module.Index0Float3Write=true
   )
   &SocketLocation(SamplePosOffset=false, Socket Index=1, Socket Reader=$System.Socket Reader, Socket Sampling Mode=::NewEnumerator2, VelocitySpeedLimit=false)
   &AddVelocity(
    Module.Use Falloff Distance=false
    Override Seed=false
    Velocity Mode=::NewEnumerator1
    Velocity Origin=$Output.SocketLocation.SampledPosition
    Velocity Speed=&MergeHelper_FloatSwitch(
     0=0
     1=&RandomRangeFloat(Minimum=55, Maximum=222)
     2=?
     3=?
     Index=$Emitter.SpawnGroup
     Num=::NewEnumerator0
    )
   )
   &CameraOffset(Camera Offset Amount=&MergeHelper_FloatSwitch(0=15, 1=5, 2=?, 3=?, Index=$Emitter.SpawnGroup, Num=::NewEnumerator0))
  ]
  particleUpdateScript [
   &ParticleState()
   &ScaleSpriteSize(
    Scale Sprite Size Mode=::NewEnumerator0
    Uniform Scale Factor=&MergeHelper_FloatSwitch(
     0=&FloatFromCurve(FloatCurve=DataInterface() {propertyValues {Curve {
         defaultValue=1
         keys [
          {interpMode=RCIM_Cubic, tangentMode=RCTM_User, time=0, value=0}
          {interpMode=RCIM_Cubic, tangentMode=RCTM_User, time=0.5, value=1}
          {tangentMode=RCTM_User, time=1, value=0}
         ]
        }}})
     1=&FloatFromCurve(FloatCurve=DataInterface() {propertyValues {Curve {
         defaultValue=1
         keys [
          {interpMode=RCIM_Cubic, tangentMode=RCTM_User, time=0, value=0}
          {arriveTangent=0.0075, interpMode=RCIM_Cubic, leaveTangent=0.0075, tangentMode=RCTM_User, time=0.25, value=1}
          {interpMode=RCIM_Cubic, tangentMode=RCTM_User, time=0.75, value=1}
          {tangentMode=RCTM_User, time=1, value=0}
         ]
        }}})
     2=?
     3=?
     Index=$Emitter.SpawnGroup
     Num=::NewEnumerator0
    )
   )
   &Drag(Drag=&RandomRangeFloat(Minimum=2, Maximum=4), Use Linear Drag=true, Use Rotational Drag=true)
   &CurlNoiseForce(
    Bias Noise=false
    Noise Strength=&MergeHelper_FloatSwitch(0=333, 1=666, 2=?, 3=?, Index=$Emitter.SpawnGroup, Num=::NewEnumerator0)
    Pan Noise=false
   )
   &SolveForcesAndVelocity(Clamp Velocity=false, Limit Acceleration=false)
   &ScaleColor(
    Scale Alpha=&MergeHelper_FloatSwitch(
     0=&FloatFromCurve(FloatCurve=DataInterface() {propertyValues {Curve {
         defaultValue=0
         keys [
          {interpMode=RCIM_Cubic, tangentMode=RCTM_User, time=0, value=0}
          {interpMode=RCIM_Cubic, tangentMode=RCTM_User, time=0.5, value=1}
          {tangentMode=RCTM_User, time=1, value=0}
         ]
        }}})
     1=&FloatFromCurve(FloatCurve=DataInterface() {propertyValues {Curve {
         defaultValue=0
         keys [
          {interpMode=RCIM_Cubic, tangentMode=RCTM_User, time=0, value=0}
          {arriveTangent=0.0359, interpMode=RCIM_Cubic, leaveTangent=0.0359, tangentMode=RCTM_User, time=0.15, value=1}
          {interpMode=RCIM_Cubic, tangentMode=RCTM_User, time=0.5, value=1}
          {tangentMode=RCTM_User, time=1, value=0}
         ]
        }}})
     2=?
     3=?
     Index=$Emitter.SpawnGroup
     Num=::NewEnumerator0
    )
    Scale RGB=&MergeHelper_VectorSwitch(
     0=&Multiply_VectorByFloat(
      Vector=<1 1 1>
      Float=&FloatFromCurve(FloatCurve=DataInterface() {propertyValues {Curve {
          defaultValue=1
          keys [
           {interpMode=RCIM_Cubic, tangentMode=RCTM_User, time=0, value=0}
           {interpMode=RCIM_Cubic, tangentMode=RCTM_User, time=0.25, value=1}
           {tangentMode=RCTM_User, time=1, value=0}
          ]
         }}})
     )
     1=<1 1 1>
     2=?
     3=?
     Index=$Emitter.SpawnGroup
     Num=::NewEnumerator0
    )
    ScaleA=true
    ScaleRGB=true
   )
  ]
  rendererClasses [&NiagaraSpriteRendererProperties]
  renderers [
   {rendererClass=&NiagaraSpriteRendererProperties}
   {rendererClass=&NiagaraSpriteRendererProperties, rendererIndex=1}
  ]
 }
 Stars {
  emitterUpdateScript [
   &EmitterState(
    Inactive Response=::NewEnumerator1
    Life Cycle Mode=::NewEnumerator1
    Loop Behavior=::NewEnumerator1
    Loop Duration Mode=::NewEnumerator1
    UseLoopDelay=false
   )
   &SpawnRate(
    Spawn Probability=&MakeCustomFloatFromBool(Bool=$User.PowersActive, True Float=1, False Float=0)
    SpawnRate=10
    Use Spawn Probability=true
   )
  ]
  particleSpawnScript [
   &FN_SetSpawnGroup(Number of Spawn Groups=3)
   /SetVariables(Particles.RandomInt:NiagaraInt32=&RandomRangeInt(Minimum=0, Maximum=7)) {isSetParameters=true}
   &InitializeParticle(
    AdjustAlpha=false
    AdjustHue=true
    AdjustSaturation=true
    AdjustValue=true
    Color=<rgba 0.7287 2.5382 3 1>
    Color Mode=::NewEnumerator3
    Hue Shift Range:Vector2f=<-0.05 0.05>
    Lifetime Max=1.2
    Lifetime Mode=::NewEnumerator1
    Sprite Size Mode=::NewEnumerator3
    Sprite UV Mode=::NewEnumerator3
    Uniform Sprite Size=&Multiply_Float(
     A=&RandomRangeFloat(Minimum=15, Maximum=20)
     B=&SelectFloatFromArray(Direct Array Index=$Engine.QualityLevel, Float Selection Array=DataInterface() {propertyValues {FloatData [0.5,0.6,0.8,1,1]}})
    )
    UsePositionOffset=false
    Write Material Random=true
   )
   &CameraOffset(Camera Offset Amount=5)
   &AddVelocity(
    Velocity=&FN_FrontEndVector3Modifier(Gameplay=<-6 -6 -6>, Front End=<6 6 6>)
    Velocity Speed Scale=5
   )
  ]
  particleUpdateScript [
   &ParticleState()
   &SocketLocation(
    SamplePosOffset=true
    Sampled Position Offset=&Add_Vector(
     A=&MergeHelper_VectorSwitch(0=<0 20 12>, 1=<0 -20 12>, 2=<0 0 20>, 3=?, Index=&SpawnGroup(), Num=::NewEnumerator1)
     B=&RandomRangeVector(Minimum=<-1 -8 -4>, Maximum=<1 8 4>)
    )
    Socket Index=1
    Socket Reader=$System.Socket Reader
    Socket Sampling Mode=::NewEnumerator3
    VelocitySpeedLimit=false
   )
   &ScaleSpriteSize(
    Uniform Curve Scale=&Lerp_Float(
     A=1
     B=&MergeHelper_FloatSwitch(0=0, 1=0, 2=1, 3=?, Index=&SpawnGroup(), Num=::NewEnumerator1)
     Alpha=$System.Timeline001.Time
    )
    Uniform Curve Sprite Scale=DataInterface() {propertyValues {Curve {keys [
        {interpMode=RCIM_Cubic, tangentMode=RCTM_User, time=0, value=0}
        {interpMode=RCIM_Cubic, tangentMode=RCTM_User, time=0.5, value=1}
        {tangentMode=RCTM_User, time=1, value=0}
       ]}}}
   ) {module=ScaleSpriteSize001}
   &ScaleColor(
    Scale Alpha=&Multiply_Float(
     A=&Multiply_Float(
      A=1
      B=&Multiply_Float(
       A=&FloatFromCurve(
        FloatCurve=DataInterface() {propertyValues {Curve {keys [
            {arriveTangent=6.5707, interpMode=RCIM_Cubic, leaveTangent=6.5707, tangentMode=RCTM_User, time=0, value=0}
            {interpMode=RCIM_Cubic, tangentMode=RCTM_User, time=0.3, value=1}
            {tangentMode=RCTM_User, time=1, value=0}
           ]}}}
        Scale Curve=&RandomRangeFloat(Minimum=0.1, Maximum=0.5)
       )
       B=1
      )
     )
     B=1
    )
    Scale RGB=&Multiply_Vector(
     A=&VectorFromFloat(Value=&Multiply_Float(
       A=&RandomRangeFloat(Minimum=0.75, Maximum=1)
       B=&SelectFloatFromArray(Direct Array Index=$Engine.QualityLevel, Float Selection Array=DataInterface() {propertyValues {FloatData [10,5,1,1,1]}})
      ))
     B=&MakeVectorFromLinearColorRGB(Color=<rgba 1 1 1 1>)
    )
    ScaleA=true
    ScaleRGB=true
   )
   /SetVariables(
    Particles.CenterFlareColor:LinearColor=&MultiplyLinearColor(
     Linear Color=$Particles.Color
     Scale Factor=&MakeVector4FromVectorAndFloat(
      "Vector (RGB)"=&MakeVectorFromLinearColorRGB(Color=<rgba 2.4876 3.9656 5 1>)
      "Float (Alpha)"=&Multiply_Float(
       A=&FloatFromCurve(
        FloatCurve=DataInterface() {propertyValues {Curve {
           defaultValue=1
           keys [
            {arriveTangent=4.0437, interpMode=RCIM_Cubic, leaveTangent=4.0437, tangentMode=RCTM_User, time=0, value=0.1}
            {arriveTangent=-0.8452, interpMode=RCIM_Cubic, leaveTangent=-0.8452, tangentMode=RCTM_User, time=0.1418, value=0.9847}
            {arriveTangent=-0.8544, interpMode=RCIM_Cubic, leaveTangent=-0.8544, tangentMode=RCTM_User, time=1, value=0}
           ]
          }}}
        Scale Curve=&Multiply_Float(A=2, B=1)
       )
       B=&RemapRange(
        Input Value To Remap=&DistanceBetweenPositions(A=$Engine.Emitter.SimulationPosition, B=$Particles.Position)
        Input Min=0
        Input Max=300
        Output Min=1
        Output Max=0
       )
      )
     )
    )
    Particles.CenterFlareSize:Vector2f=&Multiply_Vector2(
     A=$Particles.SpriteSize
     B=&Vector2DFromFloat(Value=&Multiply_Float(
       A=0.5
       B=&Lerp_Float(
        A=1
        B=&MergeHelper_FloatSwitch(0=0, 1=0, 2=1, 3=?, Index=&SpawnGroup(), Num=::NewEnumerator1)
        Alpha=$System.Timeline001.Time
       )
      ))
    )
    Particles.CoronaFlareColor:LinearColor=&MultiplyLinearColor(
     Linear Color=$Particles.Color
     Scale Factor=&MakeVector4FromVectorAndFloat(
      "Vector (RGB)"=&MakeVectorFromLinearColorRGB(Color=<rgba 0.0411 0.0805 0.15 1>)
      "Float (Alpha)"=&Multiply_Float(
       A=&FloatFromCurve(
        FloatCurve=DataInterface() {propertyValues {Curve {
           defaultValue=1
           keys [
            {arriveTangent=4.0437, interpMode=RCIM_Cubic, leaveTangent=4.0437, tangentMode=RCTM_User, time=0, value=0.1}
            {arriveTangent=-0.8452, interpMode=RCIM_Cubic, leaveTangent=-0.8452, tangentMode=RCTM_User, time=0.1418, value=0.9847}
            {arriveTangent=-0.8544, interpMode=RCIM_Cubic, leaveTangent=-0.8544, tangentMode=RCTM_User, time=1, value=0}
           ]
          }}}
        Scale Curve=&Multiply_Float(A=1, B=1)
       )
       B=&RemapRange(
        Input Value To Remap=&DistanceBetweenPositions(A=$Engine.Emitter.SimulationPosition, B=$Particles.Position)
        Input Min=0
        Input Max=300
        Output Min=1
        Output Max=0
       )
      )
     )
    )
    Particles.CoronaFlareSize:Vector2f=&Multiply_Vector2(
     A=&Multiply_Vector2D_ByFloat(Float=&Lerp_Float(A=0.3, B=1, Alpha=$System.AudioReactivity), Vector2D=$Particles.Initial.SpriteSize)
     B=&Vector2DFromFloat(Value=0.75)
    )
   ) {isSetParameters=true}
   &Drag(Use Linear Drag=true, Use Rotational Drag=true)
   &SolveForcesAndVelocity(Clamp Velocity=false, Limit Acceleration=false)
  ]
  rendererClasses [&NiagaraSpriteRendererProperties]
  renderers [
   {rendererClass=&NiagaraSpriteRendererProperties}
   {rendererClass=&NiagaraSpriteRendererProperties, rendererIndex=1}
  ]
 }
}
summary {
 systemName=NS_Pickaxe_SprintLuckSlaw_Idle
 emitters [
  {emitterName=FanGlow, rendererClasses [&NiagaraMeshRendererProperties]}
  {emitterName=StarFlash, rendererClasses [&NiagaraSpriteRendererProperties]}
  {emitterName=Stars, rendererClasses [&NiagaraSpriteRendererProperties]}
 ]
 userVariables [
  User.bIsFrontEnd {type=NiagaraBool}
  User.bIsFrontEndPreview {type=NiagaraBool}
  User.bisSwinging {type=NiagaraBool}
  User.DebugReactivity {type=NiagaraBool}
  User.PowersActive {type=NiagaraBool}
  User.SwingSectionIndex {type=NiagaraInt32}
  User.WeakPointHitCount {type=NiagaraInt32}
 ]
}
