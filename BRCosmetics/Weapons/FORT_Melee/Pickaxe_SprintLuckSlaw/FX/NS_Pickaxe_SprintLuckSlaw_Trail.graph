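# Niagara weapon trail for the SprintLuckSlaw pickaxe, drawing swing ribbons from a single TrailTip emitter with three ribbon renderers.
# AnimTrail modules spawn, initialize and update short-lived particles from bone or socket positions.
# Updates write three colors (main, ornament, trail), with alpha fading over curves scaled by the system swing float.
# A system timeline blends main and trail colors, both tinted by TODColor, while a variant index selects trail color and opacity.
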NiagaraSystem /BRCosmetics/Weapons/FORT_Melee/Pickaxe_SprintLuckSlaw/FX/NS_Pickaxe_SprintLuckSlaw_Trail
dependencies {
 usedDataInterfaces [
  &NiagaraDataInterfaceCurve
  &NiagaraDataInterfaceColorCurve
  &NiagaraDataInterfaceSocketReader
  &NiagaraDataInterfaceArrayFloat
  &NiagaraDataInterfaceArrayInt32
 ]
 usedDynamicInputs [&MakeIntFromFloat]
 usedModules [
  &self:SystemSpawnScript.SystemScriptSource.SystemScriptGraph.NiagaraNodeAssignment_0.SetVariables_#1
  &SystemState
  &Timeline
  &self:SystemSpawnScript.SystemScriptSource.SystemScriptGraph.NiagaraNodeAssignment_1.SetVariables_#2
  &self:SystemSpawnScript.SystemScriptSource.SystemScriptGraph.NiagaraNodeAssignment_2.SetVariables_#3
  &self:SystemSpawnScript.SystemScriptSource.SystemScriptGraph.NiagaraNodeAssignment_3.SetVariables_#4
 ]
}
emitters {TrailTip {
  emitterUpdateScript [
   &EmitterState()
   &FN_SpawnAnimTrail_v2()
  ]
  particleSpawnScript [
   /Game/Effects/Niagara/Modules/AnimTrail/FN_InitializeAnimTrail/FN_InitializeAnimTrailBoneIndecies(Second Bone or Socket Index=1)
   &FN_InitializeAnimTrail(Color=<rgba 0.5256 0.5624 0.5833 1>, Lifetime=0.2) {module=FN_InitializeAnimTrail001}
  ]
  particleUpdateScript [
   &FN_UpdateAnimTrail() {module=FN_UpdateAnimTrail001}
   &ScaleColor(
    Scale Alpha=&FloatFromCurve(
     FloatCurve=DataInterface() {propertyValues {Curve {
        defaultValue=1
        keys [{interpMode=RCIM_Cubic, tangentMode=RCTM_User, time=0, value=1}, {interpMode=RCIM_Cubic, tangentMode=RCTM_User, time=0.8, value=0}]
       }}}
     Scale Curve=$System.SwingFloat
    )
    Scale RGB=&Lerp_Vector(
     A=&MakeVectorFromLinearColorRGB(Color=&LerpLinearColors(StartColor=<rgba 1 0.2865 0.6605 1>, EndColor=<rgba 0 1.2399 2 1>, LerpFactor=$System.Timeline.Float))
     B=&MakeVectorFromLinearColorRGB(Color=$System.TODColor)
     Alpha=0.5
    )
    ScaleA=true
    ScaleRGB=true
   ) {module=ScaleColor002}
   /SetVariables(Particles.OrnamentColor:LinearColor=&MakeLinearColorFromVectorAndFloat(
     "Vector (RGB)"=&MakeVectorFromLinearColorRGB(Color=&ColorFromCurve(ColorCurve=DataInterface() {propertyValues {
         AlphaCurve {keys [<at 0 1 RCIM_Cubic>,<at 1 0 RCIM_Cubic>]}
         BlueCurve {keys [<at 0 1 RCIM_Cubic>,<at 1 1 RCIM_Cubic>]}
         GreenCurve {keys [<at 0 0.6731 RCIM_Cubic>,<at 1 0.1438 RCIM_Cubic>]}
         RedCurve {keys [<at 0 0 RCIM_Cubic>,<at 1 0 RCIM_Cubic>]}
        }}))
     "Float (Alpha)"=&FloatFromCurve(
      FloatCurve=DataInterface() {propertyValues {Curve {
         defaultValue=1
         keys [{interpMode=RCIM_Cubic, tangentMode=RCTM_User, time=0, value=1}, {interpMode=RCIM_Cubic, tangentMode=RCTM_User, time=1, value=0}]
        }}}
      Scale Curve=&Multiply_Float(A=1, B=$System.SwingFloat)
     )
    )) {isSetParameters=true}
   /SetVariables(Particles.ColorTrail:LinearColor=&MakeLinearColorFromVectorAndFloat(
     "Vector (RGB)"=&Multiply_Vector(
      A=&Lerp_Vector(
       A=&MakeVectorFromLinearColorRGB(Color=&MergeHelper_ColorSwitch(0=<rgba 1 1 1 1>, 1=<rgba 0 0 0 1>, 2=<rgba 1 0.1089 0.9599 1>, 3=?, Index=$System.VariantInt, Num=::NewEnumerator1))
       B=&MakeVectorFromLinearColorRGB(Color=<rgba 0 1.3853 2 1>)
       Alpha=$System.Timeline.Float
      )
      B=&Multiply_VectorByFloat(Vector=&MakeVectorFromLinearColorRGB(Color=$System.TODColor), Float=0.5)
     )
     "Float (Alpha)"=&FloatFromCurve(
      FloatCurve=DataInterface() {propertyValues {Curve {
         defaultValue=1
         keys [{interpMode=RCIM_Cubic, tangentMode=RCTM_User, time=0, value=1}, {interpMode=RCIM_Cubic, tangentMode=RCTM_User, time=0.5, value=0}]
        }}}
      Scale Curve=&Multiply_Float(
       A=&MergeHelper_FloatSwitch(0=0.33, 1=1, 2=0.5, 3=?, Index=$System.VariantInt, Num=::NewEnumerator1)
       B=$System.SwingFloat
      )
     )
    )) {isSetParameters=true}
  ]
  rendererClasses [&NiagaraRibbonRendererProperties]
  renderers [
   {rendererClass=&NiagaraRibbonRendererProperties}
   {rendererClass=&NiagaraRibbonRendererProperties, rendererIndex=1}
   {rendererClass=&NiagaraRibbonRendererProperties, rendererIndex=2}
  ]
 }}
summary {
 systemName=NS_Pickaxe_SprintLuckSlaw_Trail
 emitters [{emitterName=TrailTip, rendererClasses [&NiagaraRibbonRendererProperties]}]
 userVariables [
  User.bIsFrontEnd {type=NiagaraBool}
  User.bIsFrontEndPreview {type=NiagaraBool}
  User.PowersActive {type=NiagaraBool}
  User.RibbonIDOffset {type=NiagaraInt32}
  User.SwingSectionIndex {type=NiagaraInt32}
  User.Variant {type=NiagaraFloat, defaultValue=0, description=""}
 ]
}
