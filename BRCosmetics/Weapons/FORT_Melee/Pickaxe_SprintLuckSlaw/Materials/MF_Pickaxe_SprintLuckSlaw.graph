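# Material function for the SprintLuckSlaw pickaxe that adds a reactive glowing design to incoming material attributes.
# GlowingStage and a Blueprint-supplied start time drive a timed transition that remaps into the radius and strength of a sphere mask in pre-skinned local space.
# That mask gates an SDF texture, with stepped red plus green forming the tinted emissive and red darkening base color.
# The Fresnel branch feeds no output.
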MaterialFunction /BRCosmetics/Weapons/FORT_Melee/Pickaxe_SprintLuckSlaw/Materials/MF_Pickaxe_SprintLuckSlaw
graph {nodes [
  n1 = GetMaterialAttributes(None=FunctionInput() {inputName=OUT_MA, inputType=FunctionInput_MaterialAttributes}) {attributeGetTypes [#1, #2]}
  n2 = TextureSample() {texture=/BRCosmetics/Weapons/FORT_Melee/Pickaxe_SprintLuckSlaw/Textures/T_SprintLuckSlaw_Pickaxe_SDF}
  n3 = Step(X=@n2.R) {constX=0.1, constY=0.1}
  n4 = LocalPosition() {localOrigin=InstancePreSkinning}
  n5 = VertexInterpolator(VS=@n4.XYZ)
  n6 = ScalarParameter() {group=Reactivity, parameterName=GlowingStage}
  n7 = LinearInterpolate(A=Subtract(A=@n6), B=@n6, Alpha=NamedRerouteUsage()) {desc=Native lerp toward}
  n8 = Constant4Vector() {constant=<rgba 0 1 0 1>}
  n9 = Constant4Vector() {constant=<rgba 1 2 1 0>}
  n10 = MaterialFunctionCall(Input=@n7, Input Low=@n9.R, Input High=@n9.R, Target Low=@n9.R, Target High=@n9.R) {materialFunction=&RemapValueRange}
  n11 = Multiply(
   A=MaterialFunctionCall(Input=@n7, Input Low=@n8.R, Input High=@n8.R, Target Low=@n8.R, Target High=@n8.R) {materialFunction=&RemapValueRange}.Result
   B=@n10.Result
  )
  n12 = Multiply(
   A=SphereMask(A=ShaderStageSwitch(PixelShader=@n5.PS, VertexShader=@n4.XYZ), B=Constant(), Radius=LinearInterpolate(Alpha=@n11) {constAlpha=1, constB=60}) {attenuationRadius=0, hardnessPercent=0}
   B=@n10.Result
  )
  n13 = Multiply(A=@n3, B=@n12) {constB=2}
  FunctionOutput(None=SetMaterialAttributes!(
    MaterialAttributes=@n1.MaterialAttributes
    BaseColor=LinearInterpolate(A=@n1.BaseColor, Alpha=@n13) {constB=0}
    EmissiveColor=NamedRerouteUsage()
   ) {attributeSetTypes [#1, #2]}) {outputName=OUT_MA}
  n14 = FunctionInput() {bUsePreviewValueAsDefault=true, inputName=Fresnel, inputType=FunctionInput_Scalar, sortPriority=3, previewValue=<1 1 1 1>}
  n15 = OneMinus(None=@n14)
  n16 = VectorParameter() {parameterName=Design_EmissiveColor, defaultValue=<rgba 0 0.8 1 1>}
  n17 = Multiply(A=@n3, B=@n16.RGB)
  Multiply(A=Multiply(A=Power(Base=@n15) {constExponent=3}) {constB=2}, B=@n17) {constB=2}
  n18 = Subtract(A=Time(), B=ScalarParameter() {defaultValue=-99999, group=Reactivity, parameterName=GameTimeFromBP, sortPriority=100})
  n19 = Add(A=ScalarParameter() {defaultValue=1, group=Reactivity, parameterName=Animation_Duration, sortPriority=101}) {constB=0}
  NamedRerouteDeclaration(None=Saturate(None=Divide(A=@n18, B=@n19))) {name=TimeNative, nodeColor=<rgba 0.2891 1 0 1>}
  NamedRerouteDeclaration(None=@n1.EmissiveColor) {name=IN_EmissiveColor, nodeColor=<rgba 0 0.8 1 1>}
  n20 = Add(A=@n17, B=Multiply(A=@n2.G, B=@n16.RGB))
  NamedRerouteDeclaration(None=Multiply(A=Multiply(A=@n20, B=@n12), B=ScalarParameter() {defaultValue=2, parameterName=Design_EmissiveIntensity}) {constB=3}) {name=OUT_EmissiveColor, nodeColor=<rgba 0 0.8 1 1>}
 ]}
